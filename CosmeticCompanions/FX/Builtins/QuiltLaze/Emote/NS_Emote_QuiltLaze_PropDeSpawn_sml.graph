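# Small Niagara system for the prop despawn effect of the QuiltLaze companion emote, built from toon-style smoke emitters.
# It contains ToonSmoke, ToonSmoke_LowMedium and ToonImpactSmoke emitters, though their renderers and internals are not shown.
# A system spawn variable assignment runs alongside SystemState, and the system uses time-of-day color and float-array selection dynamic inputs.
# A NormalizedNotifyProgress user float is exposed, defaulting to zero.
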NiagaraSystem /CosmeticCompanions/FX/Builtins/QuiltLaze/Emote/NS_Emote_QuiltLaze_PropDeSpawn_sml
dependencies {
 usedDataInterfaces [&NiagaraDataInterfaceArrayFloat]
 usedDynamicInputs [&FNTimeOfDayColor, &SelectFloatFromArray]
 usedModules [
  &SystemState
  &self:SystemSpawnScript.SystemScriptSource.SystemScriptGraph.NiagaraNodeAssignment_0.SetVariables_#1
 ]
}
emitters {
 ToonImpactSmoke {rendererClasses [], renderers []}
 ToonSmoke {rendererClasses [], renderers []}
 ToonSmoke_LowMedium {rendererClasses [], renderers []}
}
summary {
 systemName=NS_Emote_QuiltLaze_PropDeSpawn_sml
 emitters [
  {emitterName=ToonSmoke, rendererClasses []}
  {emitterName=ToonSmoke_LowMedium, rendererClasses []}
  {emitterName=ToonImpactSmoke, rendererClasses []}
 ]
 userVariables [User.NormalizedNotifyProgress {type=NiagaraFloat, defaultValue=0, description=""}]
}
