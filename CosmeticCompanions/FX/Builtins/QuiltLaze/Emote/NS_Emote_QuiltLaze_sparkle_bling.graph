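# Niagara system for a sparkle bling effect in the QuiltLaze cosmetic companion's emote.
# SparklesLWE is the only enabled emitter, alongside two disabled SparklesLWE variants, and no renderers are listed for any of them.
# System scripts use System State and a spawn-time variable assignment, with time-of-day color and float-array selection dynamic inputs.
# It exposes a NormalizedNotifyProgress float user parameter, defaulting to zero.
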NiagaraSystem /CosmeticCompanions/FX/Builtins/QuiltLaze/Emote/NS_Emote_QuiltLaze_sparkle_bling
dependencies {
 usedDataInterfaces [&NiagaraDataInterfaceArrayFloat]
 usedDynamicInputs [&FNTimeOfDayColor, &SelectFloatFromArray]
 usedModules [
  &SystemState
  &self:SystemSpawnScript.SystemScriptSource.SystemScriptGraph.NiagaraNodeAssignment_0.SetVariables_#1
 ]
}
emitters {
 SparklesLWE {rendererClasses [], renderers []}
 SparklesLWE001 {bEnabled=false, rendererClasses [], renderers []}
 SparklesLWE002 {bEnabled=false, rendererClasses [], renderers []}
}
summary {
 systemName=NS_Emote_QuiltLaze_sparkle_bling
 emitters [
  {emitterName=SparklesLWE, rendererClasses []}
  {bEnabled=false, emitterName=SparklesLWE001, rendererClasses []}
  {bEnabled=false, emitterName=SparklesLWE002, rendererClasses []}
 ]
 userVariables [User.NormalizedNotifyProgress {type=NiagaraFloat, defaultValue=0, description=""}]
}
